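# Post-process Anim Blueprint for the ComicTabby quadruped cosmetic companion, adding procedural motion to its bow bones.
# On initialization it sets IsInOutfit when the Outfit channel lacks the Stage1 variant tag, enabling a constraint of the left ear to the head.
# Spring-driven, angle-limited Anim Dynamics move the bows and strands.
# Chest bow bones are first constrained to spine, neck, shoulders and elbows, then simulated root-relative.
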AnimBlueprint /CosmeticCompanions/Assets/Quadruped/ComicTabby/Meshes/ComicTabby_Quadruped_Post_AnimBP
target_skeleton=&/ComicTabby_Quadruped_Skeleton
graphs [
 AnimGraph {nodes [
   n1 = "Constraint - ear_01_l (P)"(
    ComponentPose=Local To Component(LocalPose=Use cached pose 'Start_Cache_1'())
    Alpha="To Float (Boolean)"(InBool=Get IsInOutfit() -> (IsInOutfit:bool))
    ConstraintWeights_0=1
   ) {boneToModify {boneName=ear_01_l}, constraintSetup [{targetBone {boneName=head}}], constraintWeights [1]}
   n2 = Anim Dynamics - Bone: deformer_bow_l(ComponentPose=@n1, Alpha=1, ExternalForce="0,0,0") {
    angularDampingOverride=0.9999
    angularSpringConstant=20000
    bAngularSpring=true
    bOverrideAngularDamping=true
    bUseGravityOverride=true
    gravityScale=0.35
    boundBone {boneName=deformer_bow_l}
    componentLinearAccScale=<1 1 1>
    componentLinearVelScale=<1 1 1>
    gravityOverride {z=-980}
    physicsBodyDefinitions [{
      boundBone {boneName=deformer_bow_l}
      boxExtents=<20 40 40>
      constraintSetup {angularLimitsMax=<0 0 0>, angularLimitsMin=<0 -15 -45>, angularTarget=<1 0 0>}
      localJointOffset=<4 0 0>
     }]
   }
   n3 = Anim Dynamics - Bone: deformer_bow_r(ComponentPose=@n2, Alpha=1, ExternalForce="0,0,0") {
    angularDampingOverride=0.9999
    angularSpringConstant=20000
    bAngularSpring=true
    bOverrideAngularDamping=true
    bUseGravityOverride=true
    gravityScale=0.35
    boundBone {boneName=deformer_bow_r}
    componentLinearAccScale=<1 1 1>
    componentLinearVelScale=<1 1 1>
    gravityOverride {z=-980}
    physicsBodyDefinitions [{
      boundBone {boneName=deformer_bow_r}
      boxExtents=<20 40 40>
      constraintSetup {angularLimitsMax=<0 15 0>, angularLimitsMin=<0 -3 -45>, angularTarget=<1 0 0>}
      localJointOffset=<-4 0 0>
     }]
   }
   n4 = Anim Dynamics - Bone: deformer_bow_strand_l(ComponentPose=@n3, Alpha=1, ExternalForce="0,0,0") {
    angularDampingOverride=0.9999
    angularSpringConstant=20000
    bAngularSpring=true
    bOverrideAngularDamping=true
    bUseGravityOverride=true
    gravityScale=0.35
    boundBone {boneName=deformer_bow_strand_l}
    componentLinearAccScale=<1 1 1>
    componentLinearVelScale=<1 1 1>
    gravityOverride {z=-980}
    physicsBodyDefinitions [{
      boundBone {boneName=deformer_bow_strand_l}
      boxExtents=<20 40 40>
      constraintSetup {angularLimitsMax=<0 10 5>, angularLimitsMin=<0 -10 -45>, angularTarget=<1 0 0>}
      localJointOffset=<4 0 0>
     }]
   }
   n5 = Anim Dynamics - Bone: deformer_bow_strand2_l(ComponentPose=@n4, Alpha=1, ExternalForce="0,0,0") {
    angularDampingOverride=0.9999
    angularSpringConstant=20000
    bAngularSpring=true
    bOverrideAngularDamping=true
    bUseGravityOverride=true
    gravityScale=0.35
    boundBone {boneName=deformer_bow_strand2_l}
    componentLinearAccScale=<1 1 1>
    componentLinearVelScale=<1 1 1>
    gravityOverride {z=-980}
    physicsBodyDefinitions [{
      boundBone {boneName=deformer_bow_strand2_l}
      boxExtents=<20 40 40>
      constraintSetup {angularLimitsMax=<0 10 5>, angularLimitsMin=<0 -10 -45>, angularTarget=<1 0 0>}
      localJointOffset=<4 0 0>
     }]
   }
   n6 = Anim Dynamics - Bone: deformer_bow_strand_r(ComponentPose=@n5, Alpha=1, ExternalForce="0,0,0") {
    angularDampingOverride=0.9999
    angularSpringConstant=20000
    bAngularSpring=true
    bOverrideAngularDamping=true
    bUseGravityOverride=true
    gravityScale=0.35
    boundBone {boneName=deformer_bow_strand_r}
    componentLinearAccScale=<1 1 1>
    componentLinearVelScale=<1 1 1>
    gravityOverride {z=-980}
    physicsBodyDefinitions [{
      boundBone {boneName=deformer_bow_strand_r}
      boxExtents=<20 40 40>
      constraintSetup {angularLimitsMax=<0 20 45>, angularLimitsMin=<0 -10 -2>, angularTarget=<1 0 0>}
      localJointOffset=<-4 0 0>
     }]
   }
   n7 = Anim Dynamics - Bone: deformer_bow_strand2_r(ComponentPose=@n6, Alpha=1, ExternalForce="0,0,0") {
    angularDampingOverride=0.9999
    angularSpringConstant=20000
    bAngularSpring=true
    bOverrideAngularDamping=true
    bUseGravityOverride=true
    gravityScale=0.35
    boundBone {boneName=deformer_bow_strand2_r}
    componentLinearAccScale=<1 1 1>
    componentLinearVelScale=<1 1 1>
    gravityOverride {z=-980}
    physicsBodyDefinitions [{
      boundBone {boneName=deformer_bow_strand2_r}
      boxExtents=<20 40 40>
      constraintSetup {angularLimitsMax=<0 20 45>, angularLimitsMin=<0 -10 -2>, angularTarget=<1 0 0>}
      localJointOffset=<-4 0 0>
     }]
   }
   n8 = "Constraint - deform_chestbow (PPPPP)"!(
    ComponentPose=@n7
    Alpha=1
    ConstraintWeights_0=0.4784
    ConstraintWeights_1:float=0.2245
    ConstraintWeights_2:float=0.2218
    ConstraintWeights_3:float=0.0749
    ConstraintWeights_4:float=0.0003
   ) -> (Pose) {
    boneToModify {boneName=deform_chestbow}
    constraintSetup [
     {targetBone {boneName=spine_03}}
     {targetBone {boneName=shoulder_fr_l}}
     {targetBone {boneName=shoulder_fr_r}}
     {targetBone {boneName=neck_02}}
     {targetBone {boneName=neck_01}}
    ]
    constraintWeights [1,1,1,1,1]
   }
   n9 = "Constraint - deform_chestbow_push_l (TT)"!(ComponentPose=@n8, Alpha=1, ConstraintWeights_0=0.5, ConstraintWeights_1:float=0.5) -> (Pose) {
    boneToModify {boneName=deform_chestbow_push_l}
    constraintSetup [{transformType=Translation, targetBone {boneName=elbow_fr_l}}, {transformType=Translation, targetBone {boneName=deform_chestbow}}]
    constraintWeights [1,1]
   }
   n10 = "Constraint - deform_chestbow_push_r (TT)"!(ComponentPose=@n9, Alpha=1, ConstraintWeights_0=0.5, ConstraintWeights_1:float=0.5) -> (Pose) {
    boneToModify {boneName=deform_chestbow_push_r}
    constraintSetup [{transformType=Translation, targetBone {boneName=elbow_fr_r}}, {transformType=Translation, targetBone {boneName=deform_chestbow}}]
    constraintWeights [1,1]
   }
   n11 = Anim Dynamics - Bone: deform_chestbow_r(ComponentPose=@n10, Alpha=1, ExternalForce="0,0,0") {
    angularDampingOverride=0.9999
    angularSpringConstant=20000
    bAngularSpring=true
    bOverrideAngularDamping=true
    bUseGravityOverride=true
    gravityScale=0.35
    simulationSpace=RootRelative
    boundBone {boneName=deform_chestbow_r}
    componentLinearAccScale=<1 1 1>
    componentLinearVelScale=<1 1 1>
    gravityOverride {z=-980}
    physicsBodyDefinitions [{
      boundBone {boneName=deform_chestbow_r}
      boxExtents=<20 30 30>
      constraintSetup {angularLimitsMax=<0 10 0>, angularLimitsMin=<0 -15 -30>, angularTarget=<1 0 0>}
      localJointOffset=<-5 0 0>
     }]
   }
   n12 = Anim Dynamics - Bone: deform_chestbow_l(ComponentPose=@n11, Alpha=1, ExternalForce="0,0,0") {
    angularDampingOverride=0.9999
    angularSpringConstant=20000
    bAngularSpring=true
    bOverrideAngularDamping=true
    bUseGravityOverride=true
    gravityScale=0.35
    simulationSpace=RootRelative
    boundBone {boneName=deform_chestbow_l}
    componentLinearAccScale=<1 1 1>
    componentLinearVelScale=<1 1 1>
    gravityOverride {z=-980}
    physicsBodyDefinitions [{
      boundBone {boneName=deform_chestbow_l}
      boxExtents=<20 40 40>
      constraintSetup {angularLimitsMax=<0 15 25>, angularLimitsMin=<0 -10 0>, angularTarget=<1 0 0>}
      localJointOffset=<4 0 0>
     }]
   }
   n13 = Anim Dynamics - Bone: deform_chestbow_push_r(ComponentPose=@n12, Alpha=1, ExternalForce="0,0,0") {
    angularDampingOverride=0.9999
    angularSpringConstant=20000
    bAngularSpring=true
    bOverrideAngularDamping=true
    bUseGravityOverride=true
    gravityScale=0.35
    simulationSpace=RootRelative
    boundBone {boneName=deform_chestbow_push_r}
    componentLinearAccScale=<1 1 1>
    componentLinearVelScale=<1 1 1>
    gravityOverride {z=-980}
    physicsBodyDefinitions [{
      boundBone {boneName=deform_chestbow_push_r}
      boxExtents=<20 30 30>
      constraintSetup {angularLimitsMax=<0 10 0>, angularLimitsMin=<0 -15 -30>, angularTarget=<1 0 0>}
      localJointOffset=<-5 0 0>
     }]
   }
   n14 = Anim Dynamics - Bone: deform_chestbow_push_l(ComponentPose=@n13, Alpha=1, ExternalForce="0,0,0") {
    angularDampingOverride=0.9999
    angularSpringConstant=20000
    bAngularSpring=true
    bOverrideAngularDamping=true
    bUseGravityOverride=true
    gravityScale=0.35
    simulationSpace=RootRelative
    boundBone {boneName=deform_chestbow_push_l}
    componentLinearAccScale=<1 1 1>
    componentLinearVelScale=<1 1 1>
    gravityOverride {z=-980}
    physicsBodyDefinitions [{
      boundBone {boneName=deform_chestbow_push_l}
      boxExtents=<20 40 40>
      constraintSetup {angularLimitsMax=<0 15 25>, angularLimitsMin=<0 -10 0>, angularTarget=<1 0 0>}
      localJointOffset=<4 0 0>
     }]
   }
   Output Pose(Result=Component To Local(ComponentPose=@n14))
   Save cached pose 'Start_Cache_1'(Pose=Input Pose()) {cacheName=Start_Cache_1}
  ]}
 EventGraph {nodes [
   Event BlueprintInitializeAnimation()
   n1 = GetMainAnimInstance!(self:Companion Post Process Anim Instance Object Reference) -> (ReturnValue:Fort Player Companion Anim Instance Object Reference)
   > Set IsInOutfit(IsInOutfit:bool=InvertBool!(bValue:bool=HasVariantTagForChannel!(
      self:Fort Player Companion Anim Instance Object Reference=@n1
      ChannelTag:Gameplay Tag Structure=(TagName="Cosmetics.Variant.Channel.Outfit")
      VariantTag:Gameplay Tag Structure=(TagName="Cosmetics.Variant.Property.Stage1")
     ) -> (ReturnValue:bool)) -> (ReturnValue:bool)) -> (Output_Get:bool)
  ]}
]
variables [
 IsInOutfit {type=boolean}
 PoseAsset {type=PoseAsset, default=/CosmeticCompanions/Automation_Template/companion_pose_asset_template}
]
